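# Physics asset for the Flower Skeleton backpack cosmetic, defining bodies and joints for its secondary motion.
# A box on the pack bone anchors capsules on roses, petals and spine bones, each constrained directly to the pack.
# The tail and left/right upper and lower vines form chains from the pack with narrow angular limits, tightening toward the tail tip.
# The left upper vine root alone rotates freely, unlike its limited right counterpart.
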PhysicsAsset /BRCosmetics/Accessories/FORT_Backpacks/Backpack_M_MED_Flower_Skeleton/Meshes/M_MED_Flower_Skeleton_Pack_Physics
bodies {
 lowerrose_l [{shapeType=Capsule, length=1.3315, radius=1.25, center=<0 2 0>, rotation=<rot 0 0 -90>}]
 lowerrose_petal_01_l [{shapeType=Capsule, length=0.75, radius=0.5, center=<-0.2708 0.3707 1.5>}]
 lowerrose_petal_01_r [{shapeType=Capsule, length=1, radius=1.25, center=<0.3395 -0.3016 1.5>}]
 lowerrose_petal_02_l [{shapeType=Capsule, length=1, radius=1, center=<-0.2012 0.5228 1.5>}]
 lowerrose_petal_02_r [{shapeType=Capsule, length=1, radius=0.5, center=<-0.1422 -0.2431 1.5>}]
 lowerrose_petal_03_l [{shapeType=Capsule, length=1, radius=0.75, center=<0 0.5602 1.5>}]
 lowerrose_petal_03_r [{shapeType=Capsule, length=1, radius=1, center=<-0.5131 -0.1758 1.5>}]
 lowerrose_r [{shapeType=Capsule, length=1.3315, radius=1.25, center=<0 -1.2339 1>, rotation=<rot 0 0 -90>}]
 lowervine_01_l [{shapeType=Capsule, length=2.3373, radius=1.0625, center=<-0.3442 0.1164 -0.0739>, rotation=<rot 0 -82.7898 -79.8653>}]
 lowervine_01_r [{shapeType=Capsule, length=2.3314, radius=1.0483, center=<0.3584 -0.1196 0.0706>, rotation=<rot 0 -82.8785 -79.6004>}]
 lowervine_02_l [{shapeType=Capsule, length=1.7738, radius=1.0433, center=<-1.6742 -0.1482 0.1422>, rotation=<rot 0 90.7725 -75.2475>}]
 lowervine_02_r [{shapeType=Capsule, length=1.898, radius=1.0312, center=<1.7822 0.1273 -0.0784>, rotation=<rot 0 88.548 -71.8361>}]
 lowervine_03_l [{shapeType=Capsule, length=2.1804, radius=1.4729, center=<-2.0214 0.1336 0.3983>, rotation=<rot 0 85.9717 -87.9541>}]
 lowervine_03_r [{shapeType=Capsule, length=2.544, radius=1.5057, center=<1.6996 -0.0686 -0.3837>, rotation=<rot 0 85.0557 -86.9747>}]
 lowervine_04_l [{shapeType=Capsule, length=3.3424, radius=1.2256, center=<-3.2319 0.0558 -0.4365>, rotation=<rot 0 91.8595 -70.737>}]
 lowervine_04_r [{shapeType=Capsule, length=3.3368, radius=1.212, center=<3.2231 -0.0389 0.4311>, rotation=<rot 0 91.9217 -70.714>}]
 middle_01_petal_01_l [{shapeType=Capsule, length=1, radius=1, center=<-0.0817 0.4784 -1.5>}]
 middle_01_petal_01_r [{shapeType=Capsule, length=1, radius=1.5, center=<0.4969 -0.2095 1.5>}]
 middle_01_petal_02_l [{shapeType=Capsule, length=1, radius=1.25, center=<-0.1961 0.5248 -1.5>}]
 middle_01_petal_02_r [{shapeType=Capsule, length=1, radius=0.75, center=<-0.1314 -0.2347 1.5>}]
 middle_02_petal_01_l [{shapeType=Capsule, length=1, radius=0.75, center=<-0.3008 0.3145 -1.5>}]
 middle_02_petal_01_r [{shapeType=Capsule, length=1, radius=0.5, center=<0.5174 -0.0522 1.5>}]
 middle_02_petal_02_l [{shapeType=Capsule, length=1, radius=1.25, center=<-0.1182 0.4735 -1.5>}]
 middle_02_petal_02_r [{shapeType=Capsule, length=1, radius=0.5, center=<-0.0436 -0.4276 1.5>}]
 middlerose_01_l [{shapeType=Capsule, length=2.3053, radius=1, center=<0 2 0>, rotation=<rot 0 0 -90>}]
 middlerose_01_r [{shapeType=Capsule, length=2.5, radius=1.75, center=<0.4904 -1.6901 0>, rotation=<rot 0 0 -90>}]
 middlerose_02_l [{shapeType=Capsule, length=1.9812, radius=1.25, center=<0 2 0>, rotation=<rot 0 0 -90>}]
 middlerose_02_r [{shapeType=Capsule, length=0.4812, radius=1.25, center=<-0.0001 -0.803 0.3857>, rotation=<rot 0 0 -90>}]
 pack [{shapeType=Box, extentX=5.346, extentY=15.2205, extentZ=27.1364, center=<-0.0033 6.8407 -20.7423>, rotation=<rot -0.9277 -90 -0.0163>}]
 spine_01 [{shapeType=Capsule, length=1, radius=1, center=<0 -0.1 -0.1>, rotation=<rot -90 0 0>}]
 spine_02 [{shapeType=Capsule, length=1, radius=1, center=<0 -0.1 -0.1>}]
 spine_03 [{shapeType=Capsule, length=1, radius=1, center=<0 -0.1 -0.1>}]
 spine_04 [{shapeType=Capsule, length=1, radius=1, center=<0 -0.1 -0.1>}]
 spine_05 [{shapeType=Capsule, length=1, radius=1, center=<0 -0.1 -0.1>}]
 tail_01 [{shapeType=Capsule, length=2.7131, radius=2.5937, center=<0.5 0 0>, rotation=<rot 90 0 0>}]
 tail_02 [{shapeType=Capsule, length=2.4625, radius=2.311, center=<1.5 0 0>, rotation=<rot 90 0 0>}]
 tail_03 [{shapeType=Capsule, length=2.1715, radius=1.8791, center=<2.5 0 0>, rotation=<rot 90 0 0>}]
 tail_04 [{shapeType=Capsule, length=1.6413, radius=1.7977, center=<3 0 0>, rotation=<rot 90 0 0>}]
 tail_05 [{shapeType=Capsule, length=1.5491, radius=1.0087, center=<3.25 0 0>, rotation=<rot 90 0 0>}]
 upperrose_l [{shapeType=Capsule, length=1.9501, radius=1.25, center=<0 2 0>, rotation=<rot 0 0 -90>}]
 upperrose_petal_01_l [{shapeType=Capsule, length=1, radius=1, center=<0.127 0.5266 -1.5>}]
 upperrose_petal_01_r [{shapeType=Capsule, length=1, radius=0.75, center=<0.4458 -0.2259 1.5>}]
 upperrose_petal_02_l [{shapeType=Capsule, length=0.75, radius=0.75, center=<-0.0972 0.5104 -1.5>}]
 upperrose_petal_02_r [{shapeType=Capsule, length=1, radius=1.25, center=<0.5332 -0.135 1.5>}]
 upperrose_petal_03_l [{shapeType=Capsule, length=1, radius=1.25, center=<-0.0594 0.4261 -1.5>}]
 upperrose_petal_03_r [{shapeType=Capsule, length=1, radius=0.75, center=<0.2044 -0.4394 1.5>}]
 upperrose_r [{shapeType=Capsule, length=1.7001, radius=1.25, center=<0.0001 -1.3532 0.4417>, rotation=<rot 0 0 -90>}]
 uppervine_01_l [{shapeType=Capsule, length=0.505, radius=0.505, center=<-0.1746 -0.2291 0.0132>, rotation=<rot -15.0629 -90 -13.4411>}]
 uppervine_01_r [{shapeType=Capsule, length=0.505, radius=0.505, center=<0.1774 0.2219 0.007>, rotation=<rot -19.442 -90 -18.1023>}]
 uppervine_02_l [{shapeType=Capsule, length=1.0607, radius=0.7041, center=<-0.8093 -0.0677 0.2855>, rotation=<rot 10.0309 -90 -29.3737>}]
 uppervine_02_r [{shapeType=Capsule, length=1.0216, radius=0.6794, center=<0.8299 0.0749 -0.1948>, rotation=<rot 11.503 -90 -14.9425>}]
 uppervine_03_l [{shapeType=Capsule, length=2.2668, radius=1.1501, center=<-0.902 0.248 0.1826>, rotation=<rot 0 94.2664 -79.7702>}]
 uppervine_03_r [{shapeType=Capsule, length=2.4686, radius=1.1622, center=<1.1856 -0.256 -0.1559>, rotation=<rot 0 93.1758 -79.8915>}]
 uppervine_04_l [{shapeType=Capsule, length=1.1282, radius=0.8758, center=<-0.6379 0.0171 -0.6904>, rotation=<rot 0 89.3053 -66.3125>}]
 uppervine_04_r [{shapeType=Capsule, length=1.3182, radius=1.1589, center=<1.0988 -0.0341 0.7907>, rotation=<rot 0 -92.2945 -82.4119>}]
 uppervine_05_l [{shapeType=Capsule, length=1.3572, radius=0.9484, center=<-0.7525 0.0141 -0.0233>, rotation=<rot 0 93.9912 -56.7273>}]
 uppervine_05_r [{shapeType=Capsule, length=1.2537, radius=0.9627, center=<0.9761 -0.0375 0.038>, rotation=<rot 0 -88.2057 -88.9903>}]
 uppervine_06_l [{shapeType=Capsule, length=1.6875, radius=1.2457, center=<-0.6647 0.1594 0.9285>, rotation=<rot 0 -94.9176 -61.7186>}]
 uppervine_06_r [{shapeType=Capsule, length=1.374, radius=1.3157, center=<0.8148 -0.1781 -1.2157>, rotation=<rot 0 -95.9416 -167.7745>}]
}
constraints [
 {bone1Name=upperrose_petal_01_l, bone2Name=pack, swing1LimitDegrees=20, swing2LimitDegrees=20, twistLimitDegrees=5}
 {bone1Name=upperrose_petal_02_l, bone2Name=pack, swing1LimitDegrees=20, swing2LimitDegrees=20, twistLimitDegrees=5}
 {bone1Name=upperrose_petal_03_l, bone2Name=pack, swing1LimitDegrees=20, swing2LimitDegrees=20, twistLimitDegrees=5}
 {bone1Name=upperrose_petal_01_r, bone2Name=pack, swing1LimitDegrees=20, swing2LimitDegrees=20, twistLimitDegrees=5}
 {bone1Name=upperrose_petal_02_r, bone2Name=pack, swing1LimitDegrees=20, swing2LimitDegrees=20, twistLimitDegrees=5}
 {bone1Name=upperrose_petal_03_r, bone2Name=pack, swing1LimitDegrees=20, swing2LimitDegrees=20, twistLimitDegrees=5}
 {bone1Name=middle_01_petal_01_r, bone2Name=pack, swing1LimitDegrees=20, swing2LimitDegrees=20, twistLimitDegrees=5}
 {bone1Name=middle_01_petal_02_r, bone2Name=pack, swing1LimitDegrees=20, swing2LimitDegrees=20, twistLimitDegrees=5}
 {bone1Name=middle_02_petal_01_r, bone2Name=pack, swing1LimitDegrees=20, swing2LimitDegrees=20, twistLimitDegrees=5}
 {bone1Name=middle_02_petal_02_r, bone2Name=pack, swing1LimitDegrees=20, swing2LimitDegrees=20, twistLimitDegrees=5}
 {bone1Name=lowerrose_petal_01_r, bone2Name=pack, swing1LimitDegrees=20, swing2LimitDegrees=20, twistLimitDegrees=5}
 {bone1Name=lowerrose_petal_02_r, bone2Name=pack, swing1LimitDegrees=20, swing2LimitDegrees=20, twistLimitDegrees=5}
 {bone1Name=middle_01_petal_01_l, bone2Name=pack, swing1LimitDegrees=20, swing2LimitDegrees=20, twistLimitDegrees=5}
 {bone1Name=middle_01_petal_02_l, bone2Name=pack, swing1LimitDegrees=20, swing2LimitDegrees=20, twistLimitDegrees=5}
 {bone1Name=middle_02_petal_01_l, bone2Name=pack, swing1LimitDegrees=20, swing2LimitDegrees=20, twistLimitDegrees=5}
 {bone1Name=middle_02_petal_02_l, bone2Name=pack, swing1LimitDegrees=20, swing2LimitDegrees=20, twistLimitDegrees=5}
 {bone1Name=lowerrose_petal_01_l, bone2Name=pack, swing1LimitDegrees=20, swing2LimitDegrees=20, twistLimitDegrees=5}
 {bone1Name=lowerrose_petal_02_l, bone2Name=pack, swing1LimitDegrees=20, swing2LimitDegrees=20, twistLimitDegrees=5}
 {bone1Name=lowerrose_petal_03_l, bone2Name=pack, swing1LimitDegrees=20, swing2LimitDegrees=20, twistLimitDegrees=5}
 {bone1Name=spine_01, bone2Name=pack, swing1LimitDegrees=20, swing2LimitDegrees=20, twistLimitDegrees=5}
 {bone1Name=spine_02, bone2Name=pack, swing1LimitDegrees=20, swing2LimitDegrees=20, twistLimitDegrees=5}
 {bone1Name=spine_03, bone2Name=pack, swing1LimitDegrees=20, swing2LimitDegrees=20, twistLimitDegrees=5}
 {bone1Name=spine_04, bone2Name=pack, swing1LimitDegrees=20, swing2LimitDegrees=20, twistLimitDegrees=5}
 {bone1Name=spine_05, bone2Name=pack, swing1LimitDegrees=20, swing2LimitDegrees=20, twistLimitDegrees=5}
 {bone1Name=lowerrose_petal_03_r, bone2Name=pack, swing1LimitDegrees=20, swing2LimitDegrees=20, twistLimitDegrees=5}
 {bone1Name=tail_01, bone2Name=pack, swing1LimitDegrees=15, swing2LimitDegrees=20, twistLimitDegrees=5}
 {bone1Name=tail_02, bone2Name=tail_01, swing1LimitDegrees=12, swing2LimitDegrees=20, twistLimitDegrees=5}
 {bone1Name=tail_03, bone2Name=tail_02, swing1LimitDegrees=9, swing2LimitDegrees=20, twistLimitDegrees=5}
 {bone1Name=tail_04, bone2Name=tail_03, swing1LimitDegrees=6, swing2LimitDegrees=20, twistLimitDegrees=5}
 {bone1Name=tail_05, bone2Name=tail_04, swing1LimitDegrees=6, swing2LimitDegrees=20, twistLimitDegrees=5}
 {bone1Name=upperrose_r, bone2Name=pack, swing1LimitDegrees=20, swing2LimitDegrees=20, twistLimitDegrees=5}
 {bone1Name=middlerose_01_r, bone2Name=pack, swing1LimitDegrees=20, swing2LimitDegrees=20, twistLimitDegrees=5}
 {bone1Name=middlerose_02_r, bone2Name=pack, swing1LimitDegrees=20, swing2LimitDegrees=20, twistLimitDegrees=5}
 {bone1Name=lowerrose_r, bone2Name=pack, swing1LimitDegrees=20, swing2LimitDegrees=20, twistLimitDegrees=5}
 {bone1Name=upperrose_l, bone2Name=pack, swing1LimitDegrees=20, swing2LimitDegrees=20, twistLimitDegrees=5}
 {bone1Name=middlerose_01_l, bone2Name=pack, swing1LimitDegrees=20, swing2LimitDegrees=20, twistLimitDegrees=5}
 {bone1Name=middlerose_02_l, bone2Name=pack, swing1LimitDegrees=20, swing2LimitDegrees=20, twistLimitDegrees=5}
 {bone1Name=lowerrose_l, bone2Name=pack, swing1LimitDegrees=20, swing2LimitDegrees=20, twistLimitDegrees=5}
 {
  bone1Name=uppervine_01_l
  bone2Name=pack
  swing1LimitDegrees=45
  swing1Motion=Free
  swing2LimitDegrees=45
  swing2Motion=Free
  twistLimitDegrees=45
  twistMotion=Free
 }
 {bone1Name=uppervine_02_l, bone2Name=uppervine_01_l, swing1LimitDegrees=5, swing2LimitDegrees=10, twistLimitDegrees=5}
 {bone1Name=uppervine_03_l, bone2Name=uppervine_02_l, swing1LimitDegrees=5, swing2LimitDegrees=10, twistLimitDegrees=5}
 {bone1Name=uppervine_04_l, bone2Name=uppervine_03_l, swing1LimitDegrees=5, swing2LimitDegrees=10, twistLimitDegrees=5}
 {bone1Name=uppervine_05_l, bone2Name=uppervine_04_l, swing1LimitDegrees=5, swing2LimitDegrees=10, twistLimitDegrees=5}
 {bone1Name=uppervine_06_l, bone2Name=uppervine_05_l, swing1LimitDegrees=5, swing2LimitDegrees=10, twistLimitDegrees=5}
 {bone1Name=uppervine_01_r, bone2Name=pack, swing1LimitDegrees=5, swing2LimitDegrees=10, twistLimitDegrees=5}
 {bone1Name=uppervine_02_r, bone2Name=uppervine_01_r, swing1LimitDegrees=5, swing2LimitDegrees=10, twistLimitDegrees=5}
 {bone1Name=uppervine_03_r, bone2Name=uppervine_02_r, swing1LimitDegrees=5, swing2LimitDegrees=10, twistLimitDegrees=5}
 {bone1Name=uppervine_04_r, bone2Name=uppervine_03_r, swing1LimitDegrees=5, swing2LimitDegrees=10, twistLimitDegrees=5}
 {bone1Name=uppervine_05_r, bone2Name=uppervine_04_r, swing1LimitDegrees=5, swing2LimitDegrees=10, twistLimitDegrees=5}
 {bone1Name=uppervine_06_r, bone2Name=uppervine_05_r, swing1LimitDegrees=5, swing2LimitDegrees=10, twistLimitDegrees=5}
 {bone1Name=lowervine_01_l, bone2Name=pack, swing1LimitDegrees=5, swing2LimitDegrees=10, twistLimitDegrees=5}
 {bone1Name=lowervine_02_l, bone2Name=lowervine_01_l, swing1LimitDegrees=5, swing2LimitDegrees=10, twistLimitDegrees=5}
 {bone1Name=lowervine_03_l, bone2Name=lowervine_02_l, swing1LimitDegrees=5, swing2LimitDegrees=10, twistLimitDegrees=5}
 {bone1Name=lowervine_04_l, bone2Name=lowervine_03_l, swing1LimitDegrees=5, swing2LimitDegrees=10, twistLimitDegrees=5}
 {bone1Name=lowervine_01_r, bone2Name=pack, swing1LimitDegrees=5, swing2LimitDegrees=10, twistLimitDegrees=5}
 {bone1Name=lowervine_02_r, bone2Name=lowervine_01_r, swing1LimitDegrees=5, swing2LimitDegrees=10, twistLimitDegrees=5}
 {bone1Name=lowervine_03_r, bone2Name=lowervine_02_r, swing1LimitDegrees=5, swing2LimitDegrees=10, twistLimitDegrees=5}
 {bone1Name=lowervine_04_r, bone2Name=lowervine_03_r, swing1LimitDegrees=5, swing2LimitDegrees=10, twistLimitDegrees=5}
]
